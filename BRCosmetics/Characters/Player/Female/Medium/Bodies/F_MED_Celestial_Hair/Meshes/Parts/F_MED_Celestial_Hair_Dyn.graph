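# Physics asset for the dynamic short-hair bones of the female medium Celestial Hair cosmetic part.
# Five identical small capsule bodies on the short-hair bones are each constrained directly to a larger head capsule, rather than chained to one another.
# Angular limits allow moderate swing with tight twist, and the fifth strand's wider swing axis is swapped relative to the other four.
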PhysicsAsset /BRCosmetics/Characters/Player/Female/Medium/Bodies/F_MED_Celestial_Hair/Meshes/Parts/F_MED_Celestial_Hair_Dyn
bodies {
 dyn_hair_short_01 [{shapeType=Capsule, length=3.5416, radius=0.9375, center=<2.8 0 0>, rotation=<rot 90 0 0>}]
 dyn_hair_short_02 [{shapeType=Capsule, length=3.5416, radius=0.9375, center=<2.8 0 0>, rotation=<rot 90 0 0>}]
 dyn_hair_short_03 [{shapeType=Capsule, length=3.5416, radius=0.9375, center=<2.8 0 0>, rotation=<rot 90 0 0>}]
 dyn_hair_short_04 [{shapeType=Capsule, length=3.5416, radius=0.9375, center=<2.8 0 0>, rotation=<rot 90 0 0>}]
 dyn_hair_short_05 [{shapeType=Capsule, length=3.5416, radius=0.9375, center=<2.8 0 0>, rotation=<rot 90 0 0>}]
 head [{shapeType=Capsule, length=5, radius=5, center=<1.512 0.9886 0>, rotation=<rot 90 0 0>}]
}
constraints [
 {bone1Name=dyn_hair_short_01, bone2Name=head, swing1LimitDegrees=45, swing2LimitDegrees=15, twistLimitDegrees=5}
 {bone1Name=dyn_hair_short_02, bone2Name=head, swing1LimitDegrees=45, swing2LimitDegrees=15, twistLimitDegrees=5}
 {bone1Name=dyn_hair_short_03, bone2Name=head, swing1LimitDegrees=45, swing2LimitDegrees=15, twistLimitDegrees=5}
 {bone1Name=dyn_hair_short_04, bone2Name=head, swing1LimitDegrees=45, swing2LimitDegrees=15, twistLimitDegrees=5}
 {bone1Name=dyn_hair_short_05, bone2Name=head, swing1LimitDegrees=25, swing2LimitDegrees=45, twistLimitDegrees=5}
]
